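# Physics asset for the F_MED_Lateral backpack cosmetic, giving its bag, flap, front pocket, strap, books and bookmarks capsule bodies for physical motion.
# The bag body uses three capsules and anchors the top strap, front pocket, three books and first flap segment.
# The second flap segment chains off the first and each bookmark off its book.
# Tight angular limits keep movement subtle, with the flap tip swinging widest.
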PhysicsAsset /BRCosmetics/Accessories/FORT_Backpacks/Backpack_F_MED_Lateral/Meshes/F_MED_Lateral_Pysics_Asset
bodies {
 backpack_bag [
  {shapeType=Capsule, length=23, radius=1, shapeName=backpack_bag_capsule, center=<0 -3.5 0>, rotation=<rot 45 180 100>}
  {shapeType=Capsule, length=10, radius=3, shapeName=backpack_bag_capsule, center=<1 -18.5 2>, rotation=<rot 45 0 0>}
  {shapeType=Capsule, length=6, radius=1, shapeName=backpack_bag_capsule, center=<10.3 -2.3 0>, rotation=<rot 0 8 80>}
 ]
 flap_dyn_01 [{shapeType=Capsule, length=2, radius=2, shapeName=flap_dyn_01_capsule, center=<2 0 0>, rotation=<rot 90 0 0>}]
 flap_dyn_02 [{shapeType=Capsule, length=4, radius=1.5, shapeName=flap_dyn_02_capsule, center=<5 0 1>, rotation=<rot 90 0 0>}]
 front_pocket_dyn_01 [{shapeType=Capsule, length=4, radius=3.5, shapeName=front_pocket_dyn_01_capsule, center=<3 0 1>}]
 small_book_dyn_01 [{shapeType=Capsule, length=15, radius=1, shapeName=small_book_dyn_01_capsule, center=<1.5 0 -0.5>, rotation=<rot 90 0 0>}]
 thick_book_dyn_01 [{shapeType=Capsule, length=15, radius=1.5, shapeName=thick_book_dyn_01_capsule, center=<0 0 0.5>, rotation=<rot 85 0 30>}]
 thick_book_mark_dyn_01 [{shapeType=Capsule, length=0.5, radius=0.75, shapeName=thick_book_mark_dyn_01_capsule, center=<-1.25 0 0>, rotation=<rot 90 0 0>}]
 thin_book_dyn_01 [{shapeType=Capsule, length=17, radius=1, shapeName=thin_book_dyn_01_capsule, center=<3 0 0.6>, rotation=<rot 90 0 25>}]
 thin_book_mark_dyn_01 [{shapeType=Capsule, length=1.5, radius=0.5, shapeName=thin_book_mark_dyn_01_capsule, center=<-1.5 0 0>, rotation=<rot 90 0 0>}]
 top_strap_dyn_01 [{shapeType=Capsule, length=5, radius=1, shapeName=top_strap_dyn_01_capsule, center=<2 -1 0>, rotation=<rot 90 0 90>}]
}
constraints [
 {bone1Name=thin_book_mark_dyn_01, bone2Name=thin_book_dyn_01, swing1LimitDegrees=1, swing2LimitDegrees=1, twistLimitDegrees=1}
 {bone1Name=thick_book_mark_dyn_01, bone2Name=thick_book_dyn_01, swing1LimitDegrees=1, swing2LimitDegrees=1, twistLimitDegrees=1}
 {bone1Name=flap_dyn_02, bone2Name=flap_dyn_01, swing1LimitDegrees=1, swing2LimitDegrees=25, twistLimitDegrees=1}
 {bone1Name=top_strap_dyn_01, bone2Name=backpack_bag, swing1LimitDegrees=5, swing2LimitDegrees=10, twistLimitDegrees=1}
 {bone1Name=front_pocket_dyn_01, bone2Name=backpack_bag, swing1LimitDegrees=5, swing2LimitDegrees=5, twistLimitDegrees=1}
 {bone1Name=thin_book_dyn_01, bone2Name=backpack_bag, swing1LimitDegrees=2, swing2LimitDegrees=5, twistLimitDegrees=1}
 {bone1Name=thick_book_dyn_01, bone2Name=backpack_bag, swing1LimitDegrees=2, swing2LimitDegrees=5, twistLimitDegrees=1}
 {bone1Name=small_book_dyn_01, bone2Name=backpack_bag, swing1LimitDegrees=2, swing2LimitDegrees=5, twistLimitDegrees=1}
 {bone1Name=flap_dyn_01, bone2Name=backpack_bag, swing1LimitDegrees=1, swing2LimitDegrees=5, twistLimitDegrees=1}
]
