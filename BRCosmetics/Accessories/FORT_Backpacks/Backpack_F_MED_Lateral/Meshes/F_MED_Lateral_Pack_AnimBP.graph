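# Animation Blueprint for the F_MED_Lateral backpack cosmetic, adding secondary physics motion to the pack.
# From the reference pose, a RigidBody node simulates in root-bone space, then chained Anim Dynamics nodes swing the zipper pulls and spring two bookmarks.
# The event graph sets aim and skydive weights each update.
# RigidBody alpha always takes the skydive weight, reduced while skydiving, because its selector is fixed off the ADS value.
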AnimBlueprint /BRCosmetics/Accessories/FORT_Backpacks/Backpack_F_MED_Lateral/Meshes/F_MED_Lateral_Pack_AnimBP
target_skeleton=&/F_MED_Lateral_Pack_Skeleton
graphs [
 AnimGraph {nodes [
   n1 = RigidBody(
    ComponentPose=Local To Component(LocalPose=Local Space Ref Pose())
    Alpha=SelectFloat(
     A=Get ADS_Alpha_Correction_Float_Set() -> (ADS_Alpha_Correction_Float_Set:double)
     B=Get Sky Diving Float Set() -> (Sky Diving Float Set:double)
     bPickA=false
    )
   ) {
    alpha=0
    bClampLinearTranslationLimitToRefPose=true
    bOverrideWorldGravity=true
    overridePhysicsAsset=&/F_MED_Lateral_Pysics_Asset
    simulationSpace=BaseBoneSpace
    alphaBoolBlend {blendOption=Linear}
    baseBoneRef {boneName=root}
    componentAppliedLinearAccClamp=<3000 3000 3000>
    componentLinearAccScale=<0.3 0.3 0.45>
    componentLinearVelScale=<0.15 0.15 0.225>
    overrideWorldGravity {z=-980}
   }
   n2 = Anim Dynamics - Bone: front_right_zipper_dyn_01(ComponentPose=@n1, Alpha=1, ExternalForce="0,0,0") {
    angularDampingOverride=0.999
    bGravityOverrideInSimSpace=true
    bOverrideAngularDamping=true
    simulationSpace=RootRelative
    alphaBoolBlend {blendOption=Linear}
    boundBone {boneName=front_right_zipper_dyn_01}
    componentAppliedLinearAccClamp=<7500 7500 7500>
    componentLinearAccScale=<0.75 0.75 0.75>
    componentLinearVelScale=<0.75 0.75 0.75>
    physicsBodyDefinitions [{
      sphereCollisionRadius=0
      boundBone {boneName=front_right_zipper_dyn_01}
      boxExtents=<3 10 10>
      constraintSetup {angularLimitsMax=<2 0 7>, angularLimitsMin=<-2 -60 -7>, angularTarget=<0 0 0>}
      localJointOffset=<1.5 0 0.1>
     }]
   }
   n3 = Anim Dynamics - Bone: front_right_zipper_dyn_02(ComponentPose=@n2, Alpha=1, ExternalForce="0,0,0") {
    angularDampingOverride=0.999
    bGravityOverrideInSimSpace=true
    bOverrideAngularDamping=true
    simulationSpace=RootRelative
    alphaBoolBlend {blendOption=Linear}
    boundBone {boneName=front_right_zipper_dyn_02}
    componentAppliedLinearAccClamp=<7500 7500 7500>
    componentLinearAccScale=<0.75 0.75 0.75>
    componentLinearVelScale=<0.75 0.75 0.75>
    physicsBodyDefinitions [{
      sphereCollisionRadius=0
      boundBone {boneName=front_right_zipper_dyn_02}
      boxExtents=<3 10 10>
      constraintSetup {angularLimitsMax=<2 0 7>, angularLimitsMin=<-2 -60 -7>, angularTarget=<0 0 0>}
      localJointOffset=<1.5 0 0.1>
     }]
   }
   n4 = Anim Dynamics - Bone: front_right_zipper_dyn_03(ComponentPose=@n3, Alpha=1, ExternalForce="0,0,0") {
    angularDampingOverride=0.999
    bGravityOverrideInSimSpace=true
    bOverrideAngularDamping=true
    simulationSpace=RootRelative
    alphaBoolBlend {blendOption=Linear}
    boundBone {boneName=front_right_zipper_dyn_03}
    componentAppliedLinearAccClamp=<7500 7500 7500>
    componentLinearAccScale=<0.75 0.75 0.75>
    componentLinearVelScale=<0.75 0.75 0.75>
    physicsBodyDefinitions [{
      sphereCollisionRadius=0
      boundBone {boneName=front_right_zipper_dyn_03}
      boxExtents=<3 10 10>
      constraintSetup {angularLimitsMax=<2 0 7>, angularLimitsMin=<-2 -60 -7>, angularTarget=<0 0 0>}
      localJointOffset=<1.5 0 0.1>
     }]
   }
   n5 = Anim Dynamics - Bone: zipper_l_dyn_01(ComponentPose=@n4, Alpha=1, ExternalForce="0,0,0") {
    angularDampingOverride=0.999
    bGravityOverrideInSimSpace=true
    bOverrideAngularDamping=true
    simulationSpace=RootRelative
    alphaBoolBlend {blendOption=Linear}
    boundBone {boneName=zipper_l_dyn_01}
    componentAppliedLinearAccClamp=<7500 7500 7500>
    componentLinearAccScale=<0.75 0.75 0.75>
    componentLinearVelScale=<0.75 0.75 0.75>
    physicsBodyDefinitions [{
      sphereCollisionRadius=0
      boundBone {boneName=zipper_l_dyn_01}
      boxExtents=<3 10 10>
      constraintSetup {angularLimitsMax=<2 0 45>, angularLimitsMin=<-2 -60 0>, angularTarget=<0 0 0>}
      localJointOffset=<1.5 0 0.1>
     }]
   }
   n6 = Anim Dynamics - Bone: zipper_l_dyn_02(ComponentPose=@n5, Alpha=1, ExternalForce="0,0,0") {
    angularDampingOverride=0.999
    bGravityOverrideInSimSpace=true
    bOverrideAngularDamping=true
    simulationSpace=RootRelative
    alphaBoolBlend {blendOption=Linear}
    boundBone {boneName=zipper_l_dyn_02}
    componentAppliedLinearAccClamp=<7500 7500 7500>
    componentLinearAccScale=<0.75 0.75 0.75>
    componentLinearVelScale=<0.75 0.75 0.75>
    physicsBodyDefinitions [{
      sphereCollisionRadius=0
      boundBone {boneName=zipper_l_dyn_02}
      boxExtents=<3 10 10>
      constraintSetup {angularLimitsMax=<2 0 45>, angularLimitsMin=<-2 -60 0>, angularTarget=<0 0 0>}
      localJointOffset=<1.5 0 0.1>
     }]
   }
   n7 = Anim Dynamics - Bone: zipper_l_dyn_03(ComponentPose=@n6, Alpha=1, ExternalForce="0,0,0") {
    angularDampingOverride=0.999
    bGravityOverrideInSimSpace=true
    bOverrideAngularDamping=true
    simulationSpace=RootRelative
    alphaBoolBlend {blendOption=Linear}
    boundBone {boneName=zipper_l_dyn_03}
    componentAppliedLinearAccClamp=<7500 7500 7500>
    componentLinearAccScale=<0.75 0.75 0.75>
    componentLinearVelScale=<0.75 0.75 0.75>
    physicsBodyDefinitions [{
      sphereCollisionRadius=0
      boundBone {boneName=zipper_l_dyn_03}
      boxExtents=<3 10 10>
      constraintSetup {angularLimitsMax=<2 0 45>, angularLimitsMin=<-2 -60 0>, angularTarget=<0 0 0>}
      localJointOffset=<1.5 0 0.1>
     }]
   }
   n8 = Anim Dynamics - Bone: thick_book_mark_dyn_01(ComponentPose=@n7, Alpha=1, ExternalForce="0,0,0") {
    angularDampingOverride=0.9999
    angularSpringConstant=50
    bAngularSpring=true
    bGravityOverrideInSimSpace=true
    bOverrideAngularDamping=true
    simulationSpace=RootRelative
    alphaBoolBlend {blendOption=Linear}
    boundBone {boneName=thick_book_mark_dyn_01}
    componentLinearAccScale=<1 1 1>
    componentLinearVelScale=<1 1 1>
    physicsBodyDefinitions [{
      sphereCollisionRadius=0
      boundBone {boneName=thick_book_mark_dyn_01}
      boxExtents=<10 20 20>
      constraintSetup {angularLimitsMax=<10 25 1>, angularLimitsMin=<-10 -25 -1>, angularTarget=<0 0 0>}
      localJointOffset=<1 0 0>
     }]
   }
   n9 = Anim Dynamics - Bone: thin_book_mark_dyn_01(ComponentPose=@n8, Alpha=1, ExternalForce="0,0,0") {
    angularDampingOverride=0.9999
    angularSpringConstant=50
    bAngularSpring=true
    bGravityOverrideInSimSpace=true
    bOverrideAngularDamping=true
    simulationSpace=RootRelative
    alphaBoolBlend {blendOption=Linear}
    boundBone {boneName=thin_book_mark_dyn_01}
    componentLinearAccScale=<1 1 1>
    componentLinearVelScale=<1 1 1>
    physicsBodyDefinitions [{
      sphereCollisionRadius=0
      boundBone {boneName=thin_book_mark_dyn_01}
      boxExtents=<10 20 20>
      constraintSetup {angularLimitsMax=<15 35 1>, angularLimitsMin=<-10 -35 -1>, angularTarget=<0 0 0>}
      localJointOffset=<1 0 0>
     }]
   }
   Output Pose(Result=Component To Local(ComponentPose=@n9)) {layerGroup=None}
  ]}
 EventGraph {nodes [
   Event BlueprintUpdateAnimation()
   > n1 = Sequence!(execute) -> (then_0, then_1, then_2:Exec)
   > Set ADS_Alpha_Correction_Float_Set(ADS_Alpha_Correction_Float_Set:double=SelectFloat(A=0.5, B=1, bPickA=Get bIsTargeting())) -> (Output_Get:double)
   n2 = Branch(execute=@n1.then_1, Condition=Get bIsSkyDiving())
   > Set Sky Diving Float Set(Sky Diving Float Set:double=0.6) -> (Output_Get:double)
   Set Sky Diving Float Set(execute=@n2.else, Sky Diving Float Set=1)
   TryGetPawnOwner()
  ]}
]
variables [
 ADS_Alpha_Correction_Float_Set {type=number}
 Is Jumping Float Set {type=number}
 Sky Diving Float Set {type=number}
]
